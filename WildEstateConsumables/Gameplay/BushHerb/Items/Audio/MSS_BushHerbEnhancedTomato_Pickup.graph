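# One-shot mono MetaSound source, named as the pickup sound for the BushHerb Enhanced Tomato item, using loot-collect sound class and weapon-drop attenuation.
# On play, a random modulator patch triggers a random pick from a wave array variable and supplies a pitch shift to the wave player.
# A float input boosts the player's output gain, and playback completion fires OnFinished.
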MetaSoundSource /WildEstateConsumables/Gameplay/BushHerb/Items/Audio/MSS_BushHerbEnhancedTomato_Pickup
inputs [
 AdditionalOperands {type=Float}
 UE.Source.OnPlay {type=Trigger}
]
interfaces [UE.OutputFormat.Mono:1.0, UE.Source.OneShot:1.0, UE.Source:1.1]
outputs [
 "UE.OutputFormat.Mono.Audio:0" {type=Audio}
 UE.Source.OneShot.OnFinished {type=Trigger}
]
pages [Default {
  input_defaults {AdditionalOperands=1.876, UE.Source.OnPlay=null}
  nodes [
   n1 = /FortMetasound/SoundCue/MSP_RandomModulator(In=^UE.Source.OnPlay)
   ArrayRandomGet_0 = Array.Random Get:WaveAsset:Array(Next=@n1.Out, In Array=VariableAccessor.WaveAsset:Array(Variable=InitVariable.WaveAsset:Array().Variable).Value)
   WavePlayer_0 = UE.Wave Player:Mono(Play=@ArrayRandomGet_0.On Next, Wave Asset=@ArrayRandomGet_0.Value, Pitch Shift=@n1.Pitch Out) {also {version [1,1]}}
   Multiply_0 = UE.Multiply:Audio by Float(AdditionalOperands=^AdditionalOperands, PrimaryOperand=@WavePlayer_0.Out Mono)
  ]
  links [{dst="@UE.OutputFormat.Mono.Audio:0", src=Multiply_0.Out}, {dst="@UE.Source.OneShot.OnFinished", src=WavePlayer_0.On Finished}]
 }]
properties {
 assetImportData=&self:AssetImportData
 attenuationSettings=&Fort_Weapon_Drop_Att
 maxDistance=3000
 sampleRate=48000
 soundClassObject=&Fort_Loot_Collect
 volume=0.4
}
