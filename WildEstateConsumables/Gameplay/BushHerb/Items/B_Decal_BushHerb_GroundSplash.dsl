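; Decal Blueprint for the BushHerb ground splash effect in WildEstate Consumables.
; Its construction script creates a dynamic material instance from the BushHerb splash decal material instance and stores it in the Decal_MID variable.
; It then assigns that instance to the Decal component, giving each placed splash its own material instance.
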(fn ConstructionScript
  (bind _returnvalue (Rendering|Material|CreateDynamicMaterialInstance "/WildEstateConsumables/Gameplay/BushHerb/Assets/FX/MI_Decal_WildEstate_BushHerb_Splash.MI_Decal_WildEstate_BushHerb_Splash" "None" "None" :!class /Script/Engine.KismetMaterialLibrary))
  (bind _output_get (Variables|Default|SetDecal_MID _returnvalue))
  (Rendering|Components|Decal|SetDecalMaterial (Variables|Decal|GetDecal) _output_get))
