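; Instant gameplay ability for the BushHerb consumable that heals the owning Fort pawn and converts excess healing into shield.
; On activation it splits a scalable heal amount into health up to max, with the overflow as shield.
; Where cosmetic events cannot execute, it applies both amounts through one SetByCaller gameplay effect.
; It triggers health and shield gameplay cues for positive amounts, the latter only below max shield, then ends.
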(event Ability|EventActivateAbility
  (bind _gc_health_heal (Variables|Default|GetGC%20Health%20Heal))
  (bind _gc_shieldheal (Variables|Default|GetGC_ShieldHeal))
  (bind _outfortplayerpawn (CallFunction|GetAvatarasValidFortPlayerPawn)
    (:ValidFortPlayerPawn
      (Variables|Default|SetOwningPlayer _outfortplayerpawn)
      (bind (_healthamount _shieldamount) (CallFunction|CalculateHPandShields _outfortplayerpawn))
      (Networking|CanExecuteCosmeticEvents
        (:True
          _)
        (:False
          (CallFunction|ApplyHealthShieldStatChanges _healthamount _shieldamount))) ; Cosmetic events need to run on the client for the owning player\nHealth changes can only be run on the server
      (Utilities|FlowControl|Sequence
        (:then_0
          ;[1 1 1] Health Heal cue
          (if (and (> _healthamount 0) (GameplayTags|IsGameplayTagValid _gc_health_heal))
            (Ability|ExecuteGameplayCueOnOwner _gc_health_heal)))
        (:then_1
          ;[1 1 1] Shield Heal cue
          (if (Math|Boolean|ANDBoolean (> _shieldamount 0) (GameplayTags|IsGameplayTagValid _gc_shieldheal) (< (FortPawn|GetShield (Variables|Default|GetOwningPlayer)) (FortPawn|GetMaxShield (Variables|Default|GetOwningPlayer))))
            (Ability|ExecuteGameplayCueOnOwner _gc_shieldheal)
            (else
              _))))
      (Ability|EndAbility))
    (:AvatarCastFailed
      (Ability|EndAbility))))

(event Ability|EventOnEndAbility
  (params (bWasCancelled :type Boolean)))

(fn ApplyHealthShieldStatChanges
  (params (Health :type Float%20%28single-precision%29) (Shield :type Float%20%28single-precision%29))
  (bind _returnvalue (Ability|GameplayEffect|AssignTagSetbyCallerMagnitude (Ability|MakeOutgoingGameplayEffectSpec (Variables|Default|GetGE_EffectToApply)) "(TagName=\"SetByCaller.Healing\")" Health))
  (bind _returnvalue_2 (Ability|GameplayEffect|AssignTagSetbyCallerMagnitude _returnvalue "(TagName=\"SetByCaller.Shield\")" Shield))
  (Ability|ApplyGameplayEffectSpecToOwner _returnvalue_2))

(fn CalculateHPandShields
  (params (AffectedPlayerPawn :type Fort%20Pawn%20Object%20Reference))
  (bind _row_healamount (Variables|Default|GetRow_HealAmount))
  (bind _returnvalue (ScalableFloat|GetValueatLevel _row_healamount))
  (bind _tofullhealthamount (Variables|Default|GetToFullHealthAmount))
  (Utilities|FlowControl|Sequence
    (:then_0
      ;;[1 1 1] Calculate heal and over heal
      ;[1 1 1] Calculate HP overflow into Shield
      (Variables|Default|SetToFullHealthAmount (- (FortPawn|GetMaxHealth AffectedPlayerPawn) (FortPawn|GetHealth AffectedPlayerPawn)))
      (Variables|Default|SetHealAmount (Math|Float|Min(Float) _returnvalue _tofullhealthamount))
      (Variables|Default|SetShieldAmount (Math|Float|Clamp(Float) (- _returnvalue _tofullhealthamount) 0.0 _returnvalue)))
    (:then_1
      (return (Variables|Default|GetHealAmount) (Variables|Default|GetShieldAmount)))))
