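; Thrown BushHerb consumable projectile that explodes on hit or water entry, granting healing and movement effects to pawns caught in it.
; Impact stores location and normal, re-enables instigator collision so the thrower can benefit, then explodes on authority or spawns a randomly rotated splash decal on remotes.
; On explosion it shows a sound indicator, applies gravity and heal-ability GameplayEffects to new hit pawns, then hides, disables collision and kills itself after a delay.
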(event OnEnterWaterEvent(WaterInteractionComponent)
  (params (WaterBody :type Fort%20Water%20Body%20Actor%20Object%20Reference) (WaterInteractionComponent :type Fort%20Water%20Interaction%20Component%20Object%20Reference) (bIsFirstBody :type Boolean))
  ;[1 1 1] On enter water call the the same path as the hit event but with a different hit normal and location.
  (Variables|Default|SetHitLocation (Transformation|GetActorLocation))
  (Variables|Default|SetHitNormal "0, 0,1.000000")
  (Utilities|FlowControl|DoOnce false
    (:Completed
      (Collision|SetIgnoreInstigatorCollision false) ; Need to do this or the instigator cannot benefit from the overlap heal
      (Networking|SwitchHasAuthority
        (:Authority
          (Delivery|DoExplosion))
        (:Remote
          (bind _asfort_pawn (Utilities|Casting|CastToFortPawn (stale Collision|EventHit Other))
            (:then
              (bind _returnvalue (Game|SpawnActorfromClass "/WildEstateConsumables/Gameplay/BushHerb/Items/B_Decal_BushHerb_GroundSplash.B_Decal_BushHerb_GroundSplash_C" (Math|Conversions|ToTransform(Vector) (Transformation|GetWorldLocation (Variables|/Script/Engine.Character|GetMesh _asfort_pawn))) "Undefined" "MultiplyWithRoot" 0 0))
              (Transformation|SetWorldRotation (Variables|/Script/Engine.Actor|GetRootComponent _returnvalue) (Math|Rotator|MakeRotator (+ (.roll "90.000000, 0, 0") (Math|Random|RandomFloatinRange 0.0 3.6e+02)) (.pitch "90.000000, 0, 0") (.yaw "90.000000, 0, 0")) false false))
            (:CastFailed
              (bind _returnvalue_2 (Game|SpawnActorfromClass "/WildEstateConsumables/Gameplay/BushHerb/Items/B_Decal_BushHerb_GroundSplash.B_Decal_BushHerb_GroundSplash_C" (Math|Conversions|ToTransform(Vector) (Variables|Default|GetHitLocation)) "Undefined" "MultiplyWithRoot" 0 0))
              (Transformation|SetWorldRotation (Variables|/Script/Engine.Actor|GetRootComponent _returnvalue_2) (Math|Rotator|MakeRotator (+ (.roll (Math|Rotator|MakeRotfromX (Variables|Default|GetHitNormal))) (Math|Random|RandomFloatinRange 0.0 3.6e+02)) (* (.pitch (Math|Rotator|MakeRotfromX (Variables|Default|GetHitNormal))) -1.0) (.yaw (Math|Rotator|MakeRotfromX (Variables|Default|GetHitNormal)))) false false))))))))

(event Collision|EventHit
  (params (MyComp :type Primitive%20Component%20Object%20Reference) (Other :type Actor%20Object%20Reference) (OtherComp :type Primitive%20Component%20Object%20Reference) (bSelfMoved :type Boolean) (HitLocation :type Vector :default "0, 0, 0") (HitNormal :type Vector :default "0, 0, 0") (NormalImpulse :type Vector :default "0, 0, 0") (Hit :type Hit%20Result%20Structure%20%28by%20ref%29))
  (Variables|Default|SetHitLocation HitLocation)
  (Variables|Default|SetHitNormal HitNormal)
  (Utilities|FlowControl|DoOnce false
    (:Completed
      (Collision|SetIgnoreInstigatorCollision false) ; Need to do this or the instigator cannot benefit from the overlap heal
      (Networking|SwitchHasAuthority
        (:Authority
          (Delivery|DoExplosion))
        (:Remote
          (bind _asfort_pawn (Utilities|Casting|CastToFortPawn Other)
            (:then
              (bind _returnvalue (Game|SpawnActorfromClass "/WildEstateConsumables/Gameplay/BushHerb/Items/B_Decal_BushHerb_GroundSplash.B_Decal_BushHerb_GroundSplash_C" (Math|Conversions|ToTransform(Vector) (Transformation|GetWorldLocation (Variables|/Script/Engine.Character|GetMesh _asfort_pawn))) "Undefined" "MultiplyWithRoot" 0 0))
              (Transformation|SetWorldRotation (Variables|/Script/Engine.Actor|GetRootComponent _returnvalue) (Math|Rotator|MakeRotator (+ (.roll "90.000000, 0, 0") (Math|Random|RandomFloatinRange 0.0 3.6e+02)) (.pitch "90.000000, 0, 0") (.yaw "90.000000, 0, 0")) false false))
            (:CastFailed
              (bind _returnvalue_2 (Game|SpawnActorfromClass "/WildEstateConsumables/Gameplay/BushHerb/Items/B_Decal_BushHerb_GroundSplash.B_Decal_BushHerb_GroundSplash_C" (Math|Conversions|ToTransform(Vector) (Variables|Default|GetHitLocation)) "Undefined" "MultiplyWithRoot" 0 0))
              (Transformation|SetWorldRotation (Variables|/Script/Engine.Actor|GetRootComponent _returnvalue_2) (Math|Rotator|MakeRotator (+ (.roll (Math|Rotator|MakeRotfromX (Variables|Default|GetHitNormal))) (Math|Random|RandomFloatinRange 0.0 3.6e+02)) (* (.pitch (Math|Rotator|MakeRotfromX (Variables|Default|GetHitNormal))) -1.0) (.yaw (Math|Rotator|MakeRotfromX (Variables|Default|GetHitNormal)))) false false))))))))

(event EventOnExploded
  (params (HitActors :type Array%20of%20Actor%20Object%20References) (HitResults :type Array%20of%20Hit%20Result%20Structures))
  (SoundIndicator|ShowSoundIndicator (Variables|Default|GetFortSoundIndicator) (Variables|Default|GetHitLocation) "/Game/Athena/Sounds/Items/Chug_Bomb/SoundIndicator_Chug_Bomb_Explode_Cue.SoundIndicator_Chug_Bomb_Explode_Cue" 1.0 0)
  (CallFunction|DoHealCheck HitActors)
  (Development|SetHiddeninGame (Variables|Default|GetStaticMesh) true false)
  (Components|Activation|Deactivate (Variables|Default|GetNiagara_Trail))
  (CallFunction|DisableProjectile))

(event Custom|DoHealCheck
  (params (HitActors :type Array%20of%20Actor%20Object%20References))
  (bind _self self)
  (for _array_element HitActors
    (if (not (Utilities|Array|ContainsItem (Variables|Default|GetHitActors) _array_element))
      (bind _asfort_pawn (Utilities|Casting|CastToFortPawn _array_element))
      (Utilities|Array|Add (Variables|Default|GetHitActors) _asfort_pawn)
      (bind _output_get (Variables|Default|SetAffectedPlayerPawn _asfort_pawn))
      (GameplayEffects|ApplyGameplayEffectToTarget (Ability|GetFortAbilitySystemComponent _output_get) "/WildEstateConsumables/Gameplay/BushHerb/Items/GE_Athena_BushHerb_Grav.GE_Athena_BushHerb_Grav_C" (Ability|GetFortAbilitySystemComponent _output_get) 0.0 (GameplayEffects|MakeFortEffectContext (Ability|GetFortAbilitySystemComponent _output_get) _self _self _self 0)) ; apply movement changes
      (GameplayEffects|ApplyGameplayEffectToTarget (Ability|GetFortAbilitySystemComponent _output_get) "/WildEstateConsumables/Gameplay/BushHerb/Items/GE_Athena_BushHerb_GrantHealGA.GE_Athena_BushHerb_GrantHealGA_C" (Ability|GetFortAbilitySystemComponent _output_get) 0.0 (GameplayEffects|MakeFortEffectContext (Ability|GetFortAbilitySystemComponent _output_get) _self _self _self 0)) ; apply vital changes
    )))

(event Custom|DisableProjectile
  (Collision|SetCollisionEnabled (Variables|Collision|GetCapsuleComponent) "NoCollision") ; Need this or collision with projectile can still happen after it hits
  (Utilities|Time|SetTimerbyEvent (EventDispatchers|CreateEvent :!event Custom|KillDelay) 0.5))

(event Custom|KillDelay
  (Delivery|Kill))

(fn ConstructionScript
  (|CallParentFunction))
